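# Physics asset for the Happy Ghost gadget mesh, giving its tube bones collision and jointed physical motion.
# Three capsule bodies form a chain from dyn_parent_tube, a tiny capsule at the root, through dyn_tube_01 to dyn_tube_02.
# Two constraints link neighboring bodies with tight twist and moderate swing limits, with the tip joint allowed to swing wider than the first.
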PhysicsAsset /Game/Gadgets/Assets/Happy_Ghost/Meshes/Happy_Ghost_Gadget_Physics
bodies {
 dyn_parent_tube [{shapeType=Capsule, length=1, radius=1}]
 dyn_tube_01 [{shapeType=Capsule, length=5.75, radius=5.25, center=<0.129 0 0>, rotation=<rot 90 0 0>}]
 dyn_tube_02 [{shapeType=Capsule, length=4.75, radius=2.25, center=<0 4.9427 0>}]
}
constraints [
 {bone1Name=dyn_tube_01, bone2Name=dyn_parent_tube, swing1LimitDegrees=20, swing2LimitDegrees=20, twistLimitDegrees=5}
 {bone1Name=dyn_tube_02, bone2Name=dyn_tube_01, swing1LimitDegrees=25, swing2LimitDegrees=45, twistLimitDegrees=5}
]
