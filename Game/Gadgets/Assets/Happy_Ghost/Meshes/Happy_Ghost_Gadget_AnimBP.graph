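# Animation Blueprint for the Happy Ghost gadget, which changes its pose while the player's grapple rope is active.
# Each update, it sets IsFired by checking the owning B_HappyGhost_Athena's instigator pawn for the Abilities.Grapple.RopeActive tag.
# IsFired toggles a replace-mode scale on the Slide bone of the reference pose, nearly collapsing two axes.
# A RigidBody node then adds physics in Root-bone space, using an override physics asset, custom gravity and scaled component motion.
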AnimBlueprint /Game/Gadgets/Assets/Happy_Ghost/Meshes/Happy_Ghost_Gadget_AnimBP
target_skeleton=&/SK_Happy_Ghost_Projectile_Skeleton
graphs [
 AnimGraph {nodes [
   n1 = "Transform (Modify) Bone - Bone: Slide"!(
    ComponentPose=Local To Component(LocalPose=Local Space Ref Pose())
    bAlphaBoolEnabled:bool=Get IsFired() -> (IsFired:bool)
    Scale="0.001,-1.249,0.001"
   ) -> (Pose) {alphaInputType=Bool, scaleMode=BMM_Replace, alphaBoolBlend {blendOption=Linear}, boneToModify {boneName=Slide}, scale=<0.001 -1.249 0.001>}
   n2 = RigidBody(ComponentPose=@n1) {
    bClampLinearTranslationLimitToRefPose=true
    bOverrideWorldGravity=true
    overridePhysicsAsset=&/Happy_Ghost_Gadget_Physics
    simulationSpace=BaseBoneSpace
    alphaBoolBlend {blendOption=Linear}
    baseBoneRef {boneName=Root}
    componentLinearAccScale=<0.4 0.7 0.7>
    componentLinearVelScale=<0.4 0.7 0.7>
    overrideWorldGravity {z=-950}
   }
   Output Pose(Result=Component To Local(ComponentPose=@n2)) {layerGroup=None}
  ]}
 EventGraph {nodes [
   Event BlueprintUpdateAnimation()
   > n1 = Cast To B_HappyGhost_Athena!(execute, Object=GetOwningActor()) -> (then, CastFailed, AsB Happy Ghost Athena:B Happy Ghost Athena Object Reference) {is=title K2Node_DynamicCast}
   > n2 = Cast To FortPlayerPawn(Object=Get Instigator!(self:Actor Object Reference=@n1) -> (Instigator:Pawn Object Reference))
   > n3 = Get Owned Gameplay Tags!(execute:Exec, self:Object Reference=@n2) -> (then:Exec, ReturnValue:Gameplay Tag Container Structure) {is=title K2Node_Message}
   > Set IsFired(IsFired:bool=HasTag!(
     TagContainer:"Gameplay Tag Container Structure (by ref)"=@n3
     Tag:Gameplay Tag Structure=(TagName="Abilities.Grapple.RopeActive")
     bExactMatch:bool=false
    ) -> (ReturnValue:bool)) -> (Output_Get:bool)
   TryGetPawnOwner()
  ]}
]
variables [IsFired {type=boolean}]
